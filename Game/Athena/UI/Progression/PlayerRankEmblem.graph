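# Common user widget that shows a player's rank emblem as an enlarged PlayerBanner with a level surround.
# Construction sets up the surround, selects the large texture and, by default, binds the banner to the owning player's account.
# Functions rebind the owner, apply explicit icon, color and level (also printing them), or update the surround level.
# Level updates can play Flash, a half-second horizontal flip and scale pulse.
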WidgetBlueprint /Game/Athena/UI/Progression/PlayerRankEmblem
animations [Flash {
  binding_count=1
  display_label=""
  length_seconds=0.5
  object=&self:Flash
  sequence {bindings [PlayerBanner {
     id=#1
     possessed_class=PlayerBanner_C
     tracks [{
       type=MovieScene2DTransformTrack
       sections [{
         type=MovieScene2DTransformSection
         end_frame=10
         start_frame=0
         channels [
          Translation.X {type=MovieSceneScriptingFloatChannel}
          Translation.Y {type=MovieSceneScriptingFloatChannel}
          Angle {type=MovieSceneScriptingFloatChannel}
          Scale.X {type=MovieSceneScriptingFloatChannel, keys [[0,1],[3,-1],[6,1],[8,1.1],[10,1]]}
          Scale.Y {type=MovieSceneScriptingFloatChannel, thinned_from=5, keys [[0,1],[6,1],[8,1.1],[10,1]]}
          Shear.X {type=MovieSceneScriptingFloatChannel}
          Shear.Y {type=MovieSceneScriptingFloatChannel}
         ]
        }]
      }]
    }]}
 }]
graphs [
 EventGraph {nodes [
   Event PreConstruct()
   > SetupSurround(self:Player Banner Object Reference=Get PlayerBanner() -> (PlayerBanner:Player Banner Object Reference))
   Event Construct()
   > Set bUseLargeTexture!(execute, bUseLargeTexture:bool=true, self:Player Banner Object Reference=Get PlayerBanner()) -> (then, Output_Get:bool)
   > n1 = Branch(Condition=Get BindToLocalPlayerByDefault() -> (BindToLocalPlayerByDefault:bool))
   > n2 = Cast To FortPlayerController(Object=GetOwningPlayer())
   > n3 = SetBannerOwner!(
    execute
    self:Player Banner Object Reference=Get PlayerBanner()
    OwnerId:Unique Net Id Structure=GetGameAccountId!(self:Fort Player Controller Object Reference=@n2) -> (ReturnValue:Unique Net Id Structure)
   ) -> (then)
   EnableSurround(execute=[@n3.then, @n1.else], self:Player Banner Object Reference=Get PlayerBanner())
  ]}
 SetBannerInfo {nodes [
   n1 = SetBannerInfo!() -> (then, Banner:Player Banner Info Structure) {is=title K2Node_FunctionEntry}
   n2 = Break Player Banner Info!(PlayerBannerInfo:"Player Banner Info Structure (by ref)"=@n1) -> (IconId:String, ColorId:String, Level:int) {is=title K2Node_BreakStruct}
   n3 = "To Text (String)"(InString=@n2.ColorId)
   n4 = "To Text (String)"(InString=@n2)
   > PrintString!(
    execute
    InString="To String (Text)"(InText=Format Text!(Format="LOCTEXT(\"{Icon} {Color} - {Level}\")", Icon:Text=@n4, Color:Text=@n3, Level:int=@n2.Level) -> (Result))
    bPrintToScreen
    bPrintToLog
    TextColor
    Duration=10
    Key
   ) -> (then)
   > SetBannerOwner(self=Get PlayerBanner())
   n5 = Break Player Banner Info(PlayerBannerInfo=@n1)
   > SetBannerIconAndColor!(
    execute
    self:Player Banner Object Reference=Get PlayerBanner()
    IconId:Name=String To Name(InString=@n5)
    ColorId:Name=String To Name(InString=@n5.ColorId)
   ) -> (then)
   > RefreshSurround!(execute, self:Player Banner Object Reference=Get PlayerBanner(), Level:int=@n5.Level) -> (then)
  ]}
 SetBannerOwner {nodes [
   n1 = SetBannerOwner!() -> (then, Target:Fort Player Controller Object Reference) {is=title K2Node_FunctionEntry}
   > SetBannerOwner(self=Get PlayerBanner(), OwnerId=GetGameAccountId(self=@n1)) {is=title K2Node_CallFunction}
  ]}
 SetBannerSurroundLevel {nodes [
   n1 = SetBannerSurroundLevel!() -> (then, Level:int, Animate:bool=~) {is=title K2Node_FunctionEntry}
   > RefreshSurround(self=Get PlayerBanner(), Level=@n1)
   > Branch(Condition=@n1.Animate)
   > PlayAnimation(InAnimation=Get Flash() -> (Flash:Widget Animation Object Reference))
  ]}
]
variables [BindToLocalPlayerByDefault {type=boolean, default=true}]
widgets {
 info {inheritedWidgetCount=0, parentClass=&CommonUserWidget, rootWidgetClass=&ScaleBox, widgetCount=2}
 widgets [
  ScaleBox_0 {widgetClassPath=&ScaleBox, widget=ScaleBox_0, properties {stretch=UserSpecified, userSpecifiedScale=3.2}}
  PlayerBanner {
   widgetClassPath=/Game/UI/Foundation/Profile/PlayerBanner
   bIsVariable=true
   parent=ScaleBox_0
   widget=PlayerBanner
   slot {class=ScaleBoxSlot, properties {horizontalAlignment=HAlign_Fill, verticalAlignment=VAlign_Fill}}
  }
 ]
}
